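; Melee weapon Blueprint for the Athena Blade sword, adding slash impact effects, tag-driven sounds and a holster rule.
; On impact it stores the sword tip position, waits briefly, then spawns a slash emitter at the hit, oriented along the tip's movement over that delay.
; A custom event plays Sound Library sounds by gameplay tag on the weapon mesh.
; Holstering is allowed only when the owning controller has the in-motorcycle tag.
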(event EventBeginPlay
  (|CallParentFunction) ; This node is disabled and will not be called.\nDrag off pins to build functionality.
)

(event EventTick
  (params (DeltaSeconds :type Float%20%28single-precision%29)))

(event FortWeapon|EventOnPlayImpactFX
  (params (HitResult :type Hit%20Result%20Structure%20%28by%20ref%29) (ImpactPhysicalSurface :type EPhysicalSurface%20Enum :default "SurfaceType_Default") (SpawnedPSC :type FXSystem%20Component%20Object%20Reference))
  (bind (_bblockinghit _binitialoverlap _time _distance _location _impactpoint _normal _impactnormal _physmat _hitactor) (Collision|BreakHitResult HitResult))
  (bind _swordpos (Variables|Default|GetSwordPos))
  (bind _returnvalue (Transformation|GetWorldLocation _swordpos))
  (|CallParentFunction HitResult ImpactPhysicalSurface SpawnedPSC)
  (Utilities|Casting|CastToFortPawn _hitactor)
  (Variables|Default|SetImpactTipLocation _returnvalue)
  (Utilities|FlowControl|Delay 0.04)
  (Effects|SpawnEmitteratLocation "/Game/Weapons/FORT_Melee/Effects/Impacts/P_Melee_Impact_IBlade_Slash.P_Melee_Impact_IBlade_Slash" _location (Math|Conversions|RotationFromXVector (- _returnvalue (Variables|Default|GetImpactTipLocation)))))

(event Custom|SoundLibraryHandle
  (params (Event%20Name :type Gameplay%20Tag%20Structure))
  (SoundLibrary|PlaySoundSimple (WorldSubsystems|GetSoundLibrarySubsystem) self (Utilities|Struct|MakeSimpleSoundSettings Event%20Name true false true false 1.0 1.0 0.0 0 0 "None" (Variables|Visual|GetWeaponMesh)) :!class /Script/SoundLibrary.SoundLibrarySubsystem))

(fn CanApplyHolster
  (params (HolsterId :type Name%20%28by%20ref%29 :default "None"))
  (return (Actor|ActorHasTag (Item|GetOwningController (FortWeapon|GetInventoryItem)) "Gameplay.Action.Player.InVehicle.Motorcycle")))

(fn ConstructionScript
  (|CallParentFunction))
